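; BP_SwimmingFXManager is a Blueprint that toggles global mask values for swimming foam effects through the MPC_SwimmingNFX material parameter collection.
; On BeginPlay it waits asynchronously for game state ready, then authority match ready, then a short delay, before zeroing the BiomeMask and EmissiveMask scalars.
; When destroyed, it sets both masks back to one.
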(event EventBeginPlay
  (Class|FortAsyncActionGameStateReady|GameStateReadyAsync
    (:OnReady
      (Class|FortAsyncActionAuthorityMatchReady|MatchReadyAsync
        (:OnReady
          (Utilities|FlowControl|Delay 3.0)
          (Rendering|Material|SetScalarParameterValue "/Game/Effects/Niagara/Emitters/Water/FoamEffect/MPC_SwimmingNFX.MPC_SwimmingNFX" "BiomeMask" :!class /Script/Engine.KismetMaterialLibrary)
          (Rendering|Material|SetScalarParameterValue "/Game/Effects/Niagara/Emitters/Water/FoamEffect/MPC_SwimmingNFX.MPC_SwimmingNFX" "EmissiveMask" :!class /Script/Engine.KismetMaterialLibrary))))))

(event EventDestroyed
  (Rendering|Material|SetScalarParameterValue "/Game/Effects/Niagara/Emitters/Water/FoamEffect/MPC_SwimmingNFX.MPC_SwimmingNFX" "BiomeMask" 1.0 :!class /Script/Engine.KismetMaterialLibrary)
  (Rendering|Material|SetScalarParameterValue "/Game/Effects/Niagara/Emitters/Water/FoamEffect/MPC_SwimmingNFX.MPC_SwimmingNFX" "EmissiveMask" 1.0 :!class /Script/Engine.KismetMaterialLibrary))

(fn ConstructionScript)
