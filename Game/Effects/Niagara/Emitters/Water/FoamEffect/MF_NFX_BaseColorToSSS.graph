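# Material function for the Niagara water foam effect that derives Subsurface Color from a material's Base Color.
# It reads Base Color from the incoming material attributes, multiplies it by a tint, and writes the result into Subsurface Color of the same attributes.
# The optional tint input falls back to a white "SSS" vector parameter, so Subsurface Color matches Base Color unless that parameter or input overrides it.
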MaterialFunction /Game/Effects/Niagara/Emitters/Water/FoamEffect/MF_NFX_BaseColorToSSS
graph {nodes [
  n1 = GetMaterialAttributes(None=FunctionInput() {description=BaseColor, inputName=MA, inputType=FunctionInput_MaterialAttributes, sortPriority=1}) {attributeGetTypes [#1]}
  n2 = FunctionInput(Preview=VectorParameter() {parameterName=SSS, defaultValue=<rgba 1 1 1 0>}.RGB) {bUsePreviewValueAsDefault=true, description=BaseColor, inputName=SSS_TintOvrride, sortPriority=2}
  FunctionOutput(None=SetMaterialAttributes!(MaterialAttributes=@n1.MaterialAttributes, SubsurfaceColor=Multiply(A=@n2, B=@n1.BaseColor)) {attributeSetTypes [#2]}) {outputName=MA, sortPriority=1}
 ]}
