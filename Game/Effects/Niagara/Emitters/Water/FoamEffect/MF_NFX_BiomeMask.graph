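# Material function that outputs a swamp biome mask for Niagara water foam effect materials.
# A sphere mask on world XY uses MPC_Water SwampColor's RG as center and SwampMaskHardness, with its alpha gating the radius parameter.
# It multiplies this by a particle-height fade, then scales it by BiomeMask from MPC_SwimmingNFX through a lerp from zero.
# The WaterZ, SCurve and scattering-mask nodes do not feed the output.
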MaterialFunction /Game/Effects/Niagara/Emitters/Water/FoamEffect/MF_NFX_BiomeMask
graph {nodes [
  CollectionParameter() {collection=/Water/Materials/MPC/MPC_Water, parameterName=WaterZ}
  n1 = ComponentMask(None=WorldPosition().XYZ) {g=true, r=true}
  n2 = CollectionParameter() {collection=&MPC_Water, parameterName=SwampColor}
  n3 = ComponentMask(None=@n2) {g=true, r=true}
  n4 = ComponentMask(None=@n2) {a=true}
  n5 = Multiply(
   A=ScalarParameter() {defaultValue=62000, group=Biomes, parameterName=Swamp Pondscum Radius}
   B=Saturate(None=Ceil(None=@n4)) {desc="Cancel swamp if radius <= 0"}
  )
  n6 = CollectionParameter() {collection=&MPC_Water, parameterName=SwampMaskHardness}
  n7 = SphereMask(A=@n1, B=@n3, Radius=@n5, Hardness=@n6) {desc=Swamp Pondscum Mask}
  n8 = ComponentMask(None=ParticlePositionWS()) {b=true}
  n9 = ScalarParameter() {defaultValue=3800, group=Biomes, parameterName=Z_Offset}
  n10 = Saturate(None=Divide(A=Add(A=@n8, B=@n9), B=ScalarParameter() {defaultValue=-64, group=Biomes, parameterName=Wide}))
  FunctionOutput(None=LinearInterpolate(
    A=Constant()
    B=Multiply(A=@n7, B=@n10)
    Alpha=CollectionParameter() {collection=&/MPC_SwimmingNFX, parameterName=BiomeMask}
   ))
  MaterialFunctionCall(In=@n7, Power=ScalarParameter() {defaultValue=2, group=Biomes, parameterName=BiomePower}) {materialFunction=&SCurve}
  SphereMask(A=@n1, B=@n3, Radius=@n5, Hardness=@n6) {desc=Swamp Scattering Mask}
 ]}
