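# Material function that outputs a masked emissive colour for water foam effects, placed by a world-space texture.
# Particle world position, shifted by a centre offset and scaled by texture size, gives planar XY UVs into T_WaterEmissiveMask.
# The green channel multiplies the emissive colour input, and the EmissiveMask value from MPC_SwimmingNFX blends it in from black.
# The mobile shading path outputs zero, so the glow is absent there.
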MaterialFunction /Game/Effects/Niagara/Emitters/Water/FoamEffect/MF_NFX_EmissiveMask
graph {nodes [
  n1 = Add(
   A=ShadingPathSwitch(Default=ParticlePositionWS(), Mobile=WorldPosition().XYZ) {desc=Migrated from FeatureLevelSwitch}
   B=Multiply(
    A=VectorParameter() {parameterName=CenterOffset, defaultValue=<rgba -67232 -49931.6055 0 0>}.RGB
    B=ScalarParameter() {defaultValue=-1, parameterName=Param_1}
   ) {constB=-1}
  )
  n2 = ScalarParameter() {defaultValue=40000, parameterName=TextureSize}
  n3 = Multiply(A=Divide(A=@n2), B=VectorParameter() {parameterName=Vector, defaultValue=<rgba 1 1 0 0>}.RGB)
  n4 = Divide(A=ComponentMask(None=Add(A=@n1, B=@n3)) {g=true, r=true}, B=@n2)
  n5 = TextureSample(UVs=@n4) {
   samplerSource=SSM_Clamp_WorldGroupSettings
   samplerType=SAMPLERTYPE_LinearColor
   texture=/Game/Effects/Niagara/Emitters/Water/T_WaterEmissiveMask
  }
  n6 = Multiply(
   A=@n5.G
   B=FunctionInput(Preview=Constant3Vector() {constant=<rgba 0.0312 4 3.3565 1>}) {bUsePreviewValueAsDefault=true, inputName=EmissiveOvrride}
  )
  n7 = Constant()
  FunctionOutput(None=LinearInterpolate(
    A=Constant()
    B=ShadingPathSwitch(Default=@n6, Mobile=@n7) {desc=Migrated from FeatureLevelSwitch}
    Alpha=CollectionParameter() {collection=&/MPC_SwimmingNFX, parameterName=EmissiveMask}
   ))
  ScalarParameter() {defaultValue=2}
  VertexInterpolator(VS=@n4)
 ]}
