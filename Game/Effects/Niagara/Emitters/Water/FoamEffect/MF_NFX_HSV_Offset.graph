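# Material function that shifts the HSV of BaseColor, EmissiveColor and SubsurfaceColor in a Material Attributes input, stored with the Niagara water foam effect assets.
# Each channel converts to HSV, adds HSV Offset, converts back to RGB and is multiplied by a shared tint.
# For negative offset V, the tint blends from white toward Reference Color, shifted in hue and saturation by the RGB-to-HSV converted offset, using V's magnitude as alpha.
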MaterialFunction /Game/Effects/Niagara/Emitters/Water/FoamEffect/MF_NFX_HSV_Offset
graph {
 expressions {
  MaterialExpressionAbs_0 {class=MaterialExpressionAbs, inputs [None {from=MaterialExpressionComponentMask_45, output=""}], properties {}}
  MaterialExpressionAdd_0 {
   class=MaterialExpressionAdd
   inputs [
    A {from=MaterialExpressionRgbToHsv_0}
    B {from=MaterialExpressionReroute_3}
   ]
  }
  MaterialExpressionAdd_1 {
   class=MaterialExpressionAdd
   inputs [
    A {from=MaterialExpressionRgbToHsv_1}
    B {from=MaterialExpressionReroute_3}
   ]
  }
  MaterialExpressionAdd_2 {
   class=MaterialExpressionAdd
   inputs [
    A {from=MaterialExpressionRgbToHsv_2}
    B {from=MaterialExpressionReroute_3}
   ]
  }
  MaterialExpressionAdd_3 {
   class=MaterialExpressionAdd
   inputs [
    A {from=MaterialExpressionRgbToHsv_3}
    B {from=MaterialExpressionConvert_0}
   ]
  }
  MaterialExpressionComponentMask_45 {class=MaterialExpressionComponentMask, inputs [None {from=MaterialExpressionReroute_1}], properties {b=true, desc=HSV Value}}
  MaterialExpressionConstant3Vector_1 {class=MaterialExpressionConstant3Vector, properties {constant=<rgba 1 1 1 0>}}
  MaterialExpressionConstant3Vector_2 {class=MaterialExpressionConstant3Vector, properties {desc=Reference color to apply HSV offset for tinting, constant=<rgba 0.25 1 1 0>}}
  MaterialExpressionConstant_0 {class=MaterialExpressionConstant}
  MaterialExpressionConvert_0 {
   class=MaterialExpressionConvert
   inputs [
    None {from=MaterialExpressionMaterialFunctionCall_1, output=RGB}
    None {from=MaterialExpressionMaterialFunctionCall_1, output=RGB}
    None {from=""}
   ]
   properties {
    desc="only H,S - V unchanged"
    nodeName=Make Vector3
    convertInputs [
     {
      type=Scalar
      expressionInput {
       expression=&self:MaterialExpressionMaterialFunctionCall_1
       inputName=None
       mask=0
       maskA=0
       maskB=0
       maskG=0
       maskR=0
       outputIndex=0
      }
     }
     {
      type=Scalar
      expressionInput {
       expression=&self:MaterialExpressionMaterialFunctionCall_1
       inputName=None
       mask=0
       maskA=0
       maskB=0
       maskG=0
       maskR=0
       outputIndex=1
      }
     }
     {type=Scalar, expressionInput {expression=null, inputName=None, mask=0, maskA=0, maskB=0, maskG=0, maskR=0, outputIndex=2}}
    ]
    convertOutputs [{type=Vector3}]
   }
  }
  MaterialExpressionFunctionInput_0 {
   class=MaterialExpressionFunctionInput
   inputs [Preview {from="", output=""}]
   properties {bUsePreviewValueAsDefault=true, inputName=HSV Offset}
  }
  MaterialExpressionFunctionInput_1 {
   class=MaterialExpressionFunctionInput
   inputs [Preview {from="", output=""}]
   properties {inputName=MA, inputType=FunctionInput_MaterialAttributes}
  }
  MaterialExpressionFunctionInput_2 {
   class=MaterialExpressionFunctionInput
   inputs [Preview {from=MaterialExpressionConstant3Vector_2, output=""}]
   properties {bUsePreviewValueAsDefault=true, inputName=Reference Color, previewValue=<0.25 1 1 1>}
  }
  MaterialExpressionFunctionOutput_0 {class=MaterialExpressionFunctionOutput, inputs [None {from=MaterialExpressionSetMaterialAttributes_0, output=""}]}
  MaterialExpressionGetMaterialAttributes_1 {
   class=MaterialExpressionGetMaterialAttributes
   inputs [None {from=MaterialExpressionFunctionInput_1, output=""}]
   properties {attributeGetTypes [#1, #2, #3]}
  }
  MaterialExpressionHsvToRgb_0 {class=MaterialExpressionHsvToRgb, inputs [None {from=MaterialExpressionAdd_0, output=""}], properties {}}
  MaterialExpressionHsvToRgb_1 {class=MaterialExpressionHsvToRgb, inputs [None {from=MaterialExpressionAdd_1, output=""}], properties {}}
  MaterialExpressionHsvToRgb_2 {class=MaterialExpressionHsvToRgb, inputs [None {from=MaterialExpressionAdd_2, output=""}], properties {}}
  MaterialExpressionHsvToRgb_3 {class=MaterialExpressionHsvToRgb, inputs [None {from=MaterialExpressionAdd_3, output=""}], properties {}}
  MaterialExpressionIf_0 {
   class=MaterialExpressionIf
   inputs [
    A {from=MaterialExpressionComponentMask_45, output=""}
    B {from="", output=""}
    "A > B" {from=MaterialExpressionConstant_0, output=""}
    "A == B" {from=MaterialExpressionConstant_0, output=""}
    "A < B" {from=MaterialExpressionAbs_0, output=""}
   ]
   properties {}
  }
  MaterialExpressionLinearInterpolate_0 {
   class=MaterialExpressionLinearInterpolate
   inputs [
    A {from=MaterialExpressionConstant3Vector_1}
    B {from=MaterialExpressionHsvToRgb_3}
    Alpha {from=MaterialExpressionIf_0}
   ]
  }
  MaterialExpressionMaterialFunctionCall_1 {
   class=MaterialExpressionMaterialFunctionCall
   inputs [Color {from=MaterialExpressionRgbToHsv_4}]
   properties {materialFunction=&SplitComponents}
  }
  MaterialExpressionMaterialFunctionCall_2 {
   class=MaterialExpressionMaterialFunctionCall
   inputs [Color {from=MaterialExpressionRgbToHsv_3}]
   properties {materialFunction=&SplitComponents}
  }
  MaterialExpressionMultiply_0 {
   class=MaterialExpressionMultiply
   inputs [
    A {from=MaterialExpressionHsvToRgb_0}
    B {from=MaterialExpressionReroute_2}
   ]
  }
  MaterialExpressionMultiply_1 {
   class=MaterialExpressionMultiply
   inputs [
    A {from=MaterialExpressionHsvToRgb_1}
    B {from=MaterialExpressionReroute_2}
   ]
  }
  MaterialExpressionMultiply_2 {
   class=MaterialExpressionMultiply
   inputs [
    A {from=MaterialExpressionHsvToRgb_2}
    B {from=MaterialExpressionReroute_2}
   ]
  }
  MaterialExpressionReroute_1 {class=MaterialExpressionReroute, inputs [None {from=MaterialExpressionFunctionInput_0}]}
  MaterialExpressionReroute_2 {class=MaterialExpressionReroute, inputs [None {from=MaterialExpressionLinearInterpolate_0}]}
  MaterialExpressionReroute_3 {class=MaterialExpressionReroute, inputs [None {from=MaterialExpressionFunctionInput_0}]}
  MaterialExpressionRgbToHsv_0 {class=MaterialExpressionRgbToHsv, inputs [None {from=MaterialExpressionGetMaterialAttributes_1, output=BaseColor}], properties {}}
  MaterialExpressionRgbToHsv_1 {class=MaterialExpressionRgbToHsv, inputs [None {from=MaterialExpressionGetMaterialAttributes_1, output=EmissiveColor}], properties {}}
  MaterialExpressionRgbToHsv_2 {class=MaterialExpressionRgbToHsv, inputs [None {from=MaterialExpressionGetMaterialAttributes_1, output=SubsurfaceColor}], properties {}}
  MaterialExpressionRgbToHsv_3 {class=MaterialExpressionRgbToHsv, inputs [None {from=MaterialExpressionFunctionInput_2, output=""}], properties {}}
  MaterialExpressionRgbToHsv_4 {class=MaterialExpressionRgbToHsv, inputs [None {from=MaterialExpressionReroute_1, output=""}], properties {}}
  MaterialExpressionSetMaterialAttributes_0 {
   class=MaterialExpressionSetMaterialAttributes
   inputs [
    MaterialAttributes {from=MaterialExpressionGetMaterialAttributes_1, output=MaterialAttributes}
    BaseColor {from=MaterialExpressionMultiply_0, output=""}
    EmissiveColor {from=MaterialExpressionMultiply_1, output=""}
    SubsurfaceColor {from=MaterialExpressionMultiply_2, output=""}
   ]
   properties {attributeSetTypes [#1, #2, #3]}
  }
 }
 properties {bExposeToLibrary=true, description=Shift HSV based on input values}
}
