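# Translucent, two-sided subsurface material for water foam particles, supporting Niagara dynamic parameters or curve-driven Cascade controls.
# An animated ramp mask and distorted world-space foam texture feed biome-blended curve atlas lookups for tint and opacity.
# Cubemap reflection adds highlights, while depth, camera-angle and optional boat-radius masks shape alpha.
# World position offset follows water displacement, and mobile shading drops detail distortion, normal map, reflection and depth fade.
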Material /Game/Effects/Niagara/Emitters/Water/FoamEffect/M_NFX_WaterFoam
graph {
 properties {
  bUseMaterialAttributes=true
  bUsedWithMeshParticles=true
  bUsedWithNanite=true
  bUsedWithNiagaraMeshParticles=true
  bUsedWithNiagaraSprites=true
  blendMode=BLEND_Translucent
  pixelDepthOffsetMode=PDOM_Legacy
  shadingModel=MSM_Subsurface
  twoSided=true
  usedShadingModels=Subsurface
 }
 nodes [
  n1 = ParticleColor()
  n2 = ScalarParameter() {desc="Interpret Particle RGB as HSV Offsets, and use ParticleColor Fallback instead..", parameterName=Use Particle Color RGB as HSV Offset}
  n3 = LinearInterpolate(A=@n1.RGB, B=VectorParameter() {parameterName=ParticleColor Fallback, defaultValue=<rgba 1 1 1 1>}.RGB, Alpha=@n2)
  n4 = NamedRerouteUsage()
  n5 = NamedRerouteUsage()
  n6 = TextureSampleParameter2D() {group=RampMask, parameterName=RampMask, samplerType=SAMPLERTYPE_LinearColor, texture=/Game/Effects/Niagara/MasterMaterials/Liquid/RT_Tex9}
  n7 = MaterialFunctionCall(Float3=Multiply(A=@n6.RGB, B=VectorParameter() {group=RampMask, parameterName=RampMaskMutiply, defaultValue=<rgba 1 1 1 0>}.RGB)) {materialFunction=&BreakOutFloat3Components}
  n8 = StaticSwitchParameter(True=Time(), False=NamedRerouteUsage()) {parameterName=PreviewTime}
  n9 = CurveAtlasRowParameter(CurveTime=Frac(None=Add(A=@n7.G, B=@n8))) {
   atlas=/Game/Effects/Niagara/Emitters/Water/WaterAtlas
   curve=&/Curve_FoamDisplacement
   defaultValue=4
   parameterName=DisplacementCurve
  }
  n10 = NamedRerouteDeclaration(None=@n9.R) {name=DisplacementCurve, nodeColor=<rgba 1 0.75 0 1>}
  n11 = ComponentMask(None=CameraVectorWS()) {g=true, r=true}
  n12 = Multiply(A=Multiply(A=NamedRerouteUsage(), B=ScalarParameter() {defaultValue=0.0667, parameterName=WaveIntensity}), B=NamedRerouteUsage())
  n13 = NamedRerouteUsage()
  n14 = ScalarParameter() {defaultValue=0.4, group=DisplaceTexture_Detail, parameterName=DetailDisplaceTexture_RelativeScale}
  n15 = NamedRerouteUsage()
  n16 = ScalarParameter() {defaultValue=2.2, group=DisplaceTexture_Detail, parameterName=DetailDisplaceTexture_RelativeSpeed}
  n17 = TextureSampleParameter2D(UVs=Add(A=Multiply(A=@n13, B=@n14) {constB=0.618}, B=Multiply(A=@n15, B=@n16) {constB=1.618})) {
   group=DisplaceTexture_Detail
   parameterName=DetailDisplaceTexture
   samplerType=SAMPLERTYPE_Normal
   texture=/FortWater/Textures/T_Water_TilingNormal_Waves_01
  }
  n18 = Multiply(
   A=ComponentMask(None=@n17.RGB) {g=true, r=true}
   B=ScalarParameter() {defaultValue=0.03, group=DisplaceTexture_Detail, parameterName=DetailDisplaceTextureIntensity}
  )
  n19 = NamedRerouteUsage()
  n20 = Add(A=NamedRerouteUsage(), B=Multiply(A=LinearInterpolate(A=@n18, Alpha=@n19) {constB=0}, B=NamedRerouteUsage()))
  n21 = Constant()
  n22 = Add(A=Add(A=Multiply(A=@n11, B=@n12), B=ShadingPathSwitch(Default=@n20, Mobile=@n21)), B=NamedRerouteUsage())
  n23 = ScalarParameter() {parameterName=FoamRotationAngle}
  n24 = TextureSampleParameter2D(UVs=StaticSwitchParameter(
    True=MaterialFunctionCall(UVs=@n22, "Rotation Angle (0-1)"=@n23) {materialFunction=&CustomRotator}.Rotated Values
    False=NamedRerouteDeclaration(None=@n22) {name=Foam Normal UV, nodeColor=<rgba 0.5938 0 1 1>}
   ) {group=Cascade, parameterName=UseCustomRotateFoamTexture?}) {
   parameterName=Foam_WP_AlphaTexture
   samplerType=SAMPLERTYPE_LinearColor
   texture=/Game/Athena/Apollo/Water/Textures/02_Rounded/03_WaterFoam_Veroni_base_H
  }
  n25 = Add(
   A=Subtract(A=Multiply(A=@n5, B=@n10), B=NamedRerouteUsage())
   B=Multiply(
    A=StaticSwitchParameter(True=@n24.A, False=@n24.A) {group=Cascade, parameterName=UseFoamAlphaTExture?}
    B=ScalarParameter() {defaultValue=1, parameterName=FoamAlpha Scale}
   )
  )
  n26 = Multiply(A=NamedRerouteUsage(), B=Multiply(A=Multiply(A=@n4, B=@n25)) {constB=0.5})
  n27 = CurveAtlasRowParameter(CurveTime=@n26) {
   atlas=&WaterAtlas
   curve=&/Curve_FoamTint
   defaultValue=5
   parameterName=TintCurve
  }
  n28 = CurveAtlasRowParameter(CurveTime=@n26) {
   atlas=&WaterAtlas
   curve=&/Curve_BiomeTint
   defaultValue=20
   parameterName=TintCurve_Biome
  }
  n29 = MaterialFunctionCall() {materialFunction=&/MF_NFX_BiomeMask}
  n30 = LinearInterpolate(A=@n27, B=@n28, Alpha=@n29.Result)
  n31 = LinearInterpolate(A=@n27.A, B=@n28.A, Alpha=@n29.Result)
  n32 = TextureSampleParameter2D(UVs=NamedRerouteUsage()) {
   parameterName=Foam_WP_NormalTexture
   samplerType=SAMPLERTYPE_Normal
   texture=/Game/Athena/Apollo/Water/Textures/02_Rounded/03_WaterFoam_Veroni_base_N
  }
  n33 = Constant3Vector() {constant=<rgba 0 0 1 0>}
  n34 = GetMaterialAttributes(None=SetMaterialAttributes!(
    MaterialAttributes
    Emissive Color=Multiply(A=@n3, B=@n30)
    Opacity=Multiply(A=@n1.A, B=@n31)
    Normal=ShadingPathSwitch(Default=@n32.RGB, Mobile=@n33)
   ) {attributeSetTypes [#1, #2, #3]}) {attributeGetTypes [#1, #3]}
  n35 = TextureSampleParameterCube!(UVs=ReflectionVectorWS(CustomWorldNormal=@n34.Normal), Apply View MipBias) {
   group=Reflection
   parameterName=RefractionMap
   samplerType=SAMPLERTYPE_LinearColor
   texture=/Game/Packages/Fortress_Sky/TexturesHDR/T_AthenaSkylight
  }
  n36 = Add(A=@n34.Emissive Color, B=Multiply(A=@n35.G, B=ScalarParameter() {defaultValue=1, group=Reflection, parameterName=ReflectionIntensity}))
  n37 = Constant()
  n38 = GetMaterialAttributes(None=SetMaterialAttributes!(MaterialAttributes=@n34.MaterialAttributes, Base Color=@n34.Emissive Color, Emissive Color=ShadingPathSwitch(Default=@n36, Mobile=@n37)) {attributeSetTypes [#4, #1]}) {attributeGetTypes [#2]}
  n39 = ComponentMask(None=CameraVectorWS()) {b=true}
  n40 = Saturate(None=Divide(A=Abs(None=@n39), B=ScalarParameter() {defaultValue=0.2, group=OpacityMaskControl, parameterName="CameraAngleFade(0-1)"}))
  n41 = ScalarParameter() {defaultValue=1, parameterName=AlphaContrast}
  n42 = NamedRerouteUsage()
  n43 = Multiply(
   A=MaterialFunctionCall(In=Multiply(A=@n38.Opacity, B=@n40), Contrast=Multiply(A=@n41, B=@n42)) {materialFunction=&CheapContrast}.Result
   B=ScalarParameter() {defaultValue=0.2, parameterName=MasterAlpha}
  )
  n44 = NamedRerouteUsage()
  n45 = GetMaterialAttributes(None=SetMaterialAttributes!(MaterialAttributes=@n38.MaterialAttributes, Opacity=Multiply(A=@n43, B=@n44)) {attributeSetTypes [#2]}) {attributeGetTypes [#4, #3, #2]}
  n46 = Multiply(A=VectorParameter() {parameterName=BaseColorTint, defaultValue=<rgba 0.9 0.9 0.9 0>}.RGB, B=@n45.Base Color)
  n47 = Add(
   A=Multiply(A=@n46, B=ScalarParameter() {parameterName=OverallEmissiveAdd})
   B=MaterialFunctionCall!(EmissiveOvrride) {materialFunction=&/MF_NFX_EmissiveMask}.Result
  )
  n48 = Saturate(None=@n45.Opacity)
  n49 = StaticSwitchParameter(True=Multiply(A=NamedRerouteUsage(), B=@n48), False=@n48) {parameterName=UseFoamMasking}
  n50 = NamedRerouteUsage()
  n51 = NamedRerouteUsage()
  n52 = ScalarParameter() {desc="Interpret Particle RGB as HSV Offsets, and use ParticleColor Fallback instead..", parameterName=Use Particle Color RGB as HSV Offset}
  MaterialFunctionCall!(
   MA=MaterialFunctionCall!(
    MA=MakeMaterialAttributes!(
     BaseColor=@n46
     Metallic
     Specular
     Roughness
     Anisotropy
     EmissiveColor=@n47
     Opacity=@n49
     OpacityMask=@n49
     Normal=@n50
     Tangent
     WorldPositionOffset=@n51
     SubsurfaceColor
     ClearCoat
     ClearCoatRoughness
     AmbientOcclusion
     Refraction
     CustomizedUVs_0
     CustomizedUVs_1
     CustomizedUVs_2
     CustomizedUVs_3
     CustomizedUVs_4
     CustomizedUVs_5
     CustomizedUVs_6
     CustomizedUVs_7
     PixelDepthOffset
     ShadingModel
     Displacement
    )
    SSS_TintOvrride
   ) {materialFunction=&/MF_NFX_BaseColorToSSS}.MA
   Reference Color
   HSV Offset=LinearInterpolate(A=Constant3Vector(), B=ParticleColor().RGB, Alpha=@n52)
  ) {desc=colored water, materialFunction=&/MF_NFX_HSV_Offset}
  MaterialFunctionCall(Alpha Threshold=@n49) {materialFunction=&DitherTemporalAA}
  n53 = DynamicParameter() {
   defaultValue=<rgba 0 512 1 0>
   paramNames [
    Detail Time Control / Cascade - WaterBodyIndex
    Wave Time Control / Cascade - WaterDepth
    Wave Intensity / Cascade - WaveAttenuationFactor
    Cascade - Curve Control
   ]
  }
  n54 = DynamicParameter() {parameterIndex=3, defaultValue=<rgba 0 512 1 0>, paramNames [WaterBodyIndex, WaterDepth, WaveAttenuationFactor, HSVOffset_Saturation]}
  NamedRerouteDeclaration(None=StaticSwitchParameter(True=@n53.Detail Time Control / Cascade - WaterBodyIndex, False=@n54.WaterBodyIndex) {group=Cascade, parameterName="OpenDPCurveControl(Cascade)"}) {name=DP3_Water Body Index, nodeColor=<rgba 0 0.3672 1 1>}
  NamedRerouteDeclaration(None=StaticSwitchParameter(True=@n53.Wave Time Control / Cascade - WaterDepth, False=@n54.WaterDepth) {group=Cascade, parameterName="OpenDPCurveControl(Cascade)"}) {name=DP3_Water Depth, nodeColor=<rgba 0 0.9531 1 1>}
  n55 = DynamicParameter() {
   defaultValue=<rgba 0 512 1 0>
   paramNames [
    Detail Time Control / Cascade - WaterBodyIndex
    Wave Time Control / Cascade - WaterDepth
    Wave Intensity / Cascade - WaveAttenuationFactor
    Cascade - Curve Control
   ]
  }
  n56 = CurveAtlasRowParameter(CurveTime=@n55.Cascade - Curve Control) {
   atlas=&WaterAtlas
   curve=&/Curve_CascadeDP0
   defaultValue=8
   group=Cascade
   parameterName=DP0_Curve
  }
  n57 = DynamicParameter() {
   defaultValue=<rgba 0 512 1 0>
   paramNames [
    Detail Time Control / Cascade - WaterBodyIndex
    Wave Time Control / Cascade - WaterDepth
    Wave Intensity / Cascade - WaveAttenuationFactor
    Cascade - Curve Control
   ]
  }
  NamedRerouteDeclaration(None=StaticSwitchParameter(True=@n56.G, False=@n57.Wave Time Control / Cascade - WaterDepth) {group=Cascade, parameterName="OpenDPCurveControl(Cascade)"}) {name=Wave_Time_Control, nodeColor=<rgba 0 1 0.1562 1>}
  NamedRerouteDeclaration(None=StaticSwitchParameter(True=@n56.B, False=@n57.Wave Intensity / Cascade - WaveAttenuationFactor) {group=Cascade, parameterName="OpenDPCurveControl(Cascade)"}) {name=Wave Intensity, nodeColor=<rgba 1 0 0.1875 1>}
  NamedRerouteDeclaration(None=@n7.R) {name=RampMaskValue, nodeColor=<rgba 0.4062 1 0 1>}
  n58 = ComponentMask(None=WorldPosition().XYZ) {g=true, r=true}
  n59 = AppendVector(A=NamedRerouteUsage(), B=NamedRerouteUsage())
  n60 = ScalarParameter() {defaultValue=1, parameterName=WP_TextureScaleUp}
  n61 = NamedRerouteUsage()
  NamedRerouteDeclaration(None=Divide(A=Subtract(A=@n58, B=@n59), B=Multiply(A=@n60, B=@n61))) {name=WPTexture, nodeColor=<rgba 0 0.2266 1 1>}
  n62 = Time()
  n63 = NamedRerouteUsage()
  n64 = ScalarParameter() {defaultValue=0.1, parameterName=WP_Texture_PannerSpeed}
  n65 = Constant2Vector() {g=1, r=1}
  NamedRerouteDeclaration(None=Multiply(A=StaticSwitchParameter(True=@n62, False=@n63) {parameterName=PreviewTime}, B=Multiply(A=@n64, B=@n65))) {name=From Time Control, nodeColor=<rgba 1 0 0.8672 1>}
  n66 = TextureSampleParameter2D() {
   group=OpacityMaskControl
   parameterName=DepthFadeTextureMask
   samplerType=SAMPLERTYPE_LinearColor
   texture=&HighContrastNoise
  }
  n67 = ComponentMask(None=StaticComponentMaskParameter(None=@n66.RGBA) {defaultR=true, group=OpacityMaskControl, parameterName=DepthFadeTextureMask_1ch}) {r=true}
  NamedRerouteDeclaration(None=@n67) {name=DepthFadeTexture, nodeColor=<rgba 0.1484 0 1 1>}
  n68 = DepthFade(FadeDistance=LinearInterpolate(
    A=ScalarParameter() {defaultValue=30, group=OpacityMaskControl, parameterName=DepthFadeHigh}
    B=ScalarParameter() {defaultValue=5, group=OpacityMaskControl, parameterName=DepthFadeLow}
    Alpha=@n67
   ))
  n69 = CurveAtlasRowParameter(CurveTime=@n68) {
   atlas=&WaterAtlas
   curve=&Curve_FoamDisplacement
   defaultValue=4
   parameterName=DisplacementCurve
  }
  n70 = WorldPosition()
  n71 = CameraPositionWS()
  n72 = ScalarParameter() {defaultValue=200, group=OpacityMaskControl, parameterName=CameraFade}
  n73 = Multiply(A=@n69.G, B=OneMinus(None=SphereMask(A=@n70.XYZ, B=@n71, Radius=@n72) {hardnessPercent=0}))
  NamedRerouteDeclaration(None=ShadingPathSwitch(Default=StaticSwitchParameter(True=@n68, False=@n73) {parameterName=UseCheapandSimple_Depthfade?}, Mobile=Constant() {r=1})) {name=DepthFade, nodeColor=<rgba 0.125 0 1 1>}
  n74 = DynamicParameter() {parameterIndex=3, defaultValue=<rgba 0 512 1 0>, paramNames [WaterBodyIndex, WaterDepth, WaveAttenuationFactor, HSVOffset_Saturation]}
  n75 = MaterialFunctionCall!(Ignore XY Offsets, Position, WaterBodyIndex=@n74.WaterBodyIndex, Depth=@n74.WaterBodyIndex) {materialFunction=/FortWater/Materials/Functions/WaterParticleDisplacement}
  n76 = VectorParameter() {parameterName=WPOOffset, defaultValue=<rgba 0 0 5 1>}
  n77 = Multiply(
   A=MaterialFunctionCall(
    Ignore XY Offsets=StaticBoolParameter() {defaultValue=true, parameterName=Ignore XY Gerstner Offsets}
    WaterBodyIndex=NamedRerouteUsage()
    Depth=NamedRerouteUsage()
   ) {materialFunction=&WaterParticleDisplacement}.WPO
   B=VectorParameter() {parameterName=Water Displacement Multiplier, defaultValue=<rgba 1 1 1 0>}.RGB
  )
  n78 = ObjectPositionWS()
  n79 = ParticlePositionWS()
  n80 = ComponentMask(None=Subtract(A=CameraPositionWS(), B=ShadingPathSwitch(Default=@n78, Deferred=@n79))) {b=true}
  n81 = MaterialFunctionCall(Z=Multiply(A=ScalarParameter() {defaultValue=1, parameterName=ZOffset}, B=Clamp(None=@n80) {minDefault=-1})) {materialFunction=&MakeFloat3}
  n82 = Add(
   A=StaticSwitchParameter(True=Add(A=@n77, B=@n81.Result), False=Constant()) {parameterName=UseWaterDisplacement}
   B=MaterialFunctionCall(Offset Amount=ScalarParameter() {parameterName=CameraOffset}) {materialFunction=/Engine/Functions/Engine_MaterialFunctions02/WorldPositionOffset/CameraOffset}.Clamped Camera Offset
  )
  n83 = Multiply(
   A=Multiply(A=NamedRerouteUsage(), B=ParticleColor().A)
   B=AppendVector(A=Constant2Vector(), B=ScalarParameter() {parameterName=WPO_WaveFromOpacity})
  )
  NamedRerouteDeclaration(None=StaticSwitchParameter(True=Add(A=@n75.WPO, B=@n76.RGB), False=Add(A=@n82, B=@n83)) {parameterName=UseCascadeWPOControls?}) {name=World Position Offset, nodeColor=<rgba 0 1 0.9063 1>}
  n84 = DynamicParameter() {
   defaultValue=<rgba 0 512 1 0>
   paramNames [
    Detail Time Control / Cascade - WaterBodyIndex
    Wave Time Control / Cascade - WaterDepth
    Wave Intensity / Cascade - WaveAttenuationFactor
    Cascade - Curve Control
   ]
  }
  n85 = CurveAtlasRowParameter(CurveTime=@n84.Cascade - Curve Control) {
   atlas=&WaterAtlas
   curve=&/Curve_CascadeDP2
   defaultValue=9
   group=Cascade
   parameterName=DP2_Curve
  }
  n86 = DynamicParameter() {
   parameterIndex=2
   defaultValue=<rgba 0.5 0.5 1 0>
   paramNames [FoamAlphaBaisA_Subject, FoamAlphaBaisB_Multiply, MasterAlpha, MasterAlphaContrast]
  }
  NamedRerouteDeclaration(None=StaticSwitchParameter(True=@n85.R, False=@n86.FoamAlphaBaisA_Subject) {group=Cascade, parameterName="OpenDPCurveControl(Cascade)"}) {name=DP2_FoamAlphaBias_Subject, nodeColor=<rgba 1 0.6563 0 1>}
  NamedRerouteDeclaration(None=@n45.Normal) {name=Normal, nodeColor=<rgba 1 0.1875 0 1>}
  n87 = Add(
   A=ActorPositionWS()
   B=Transform(None=VectorParameter() {group=Boat Foam Mask, parameterName=Boat Foam Mask Offset}.RGB) {transformSourceType=TRANSFORMSOURCE_Local}
  )
  n88 = WorldPosition() {worldPositionShaderOffset=WPT_ExcludeAllShaderOffsets}
  n89 = ComponentMask(None=@n87) {b=true, g=true, r=true}
  n90 = Add(
   A=Multiply(A=Subtract(A=@n88.XYZ, B=@n89), B=VectorParameter() {group=Boat Foam Mask, parameterName=ScaleMask, defaultValue=<rgba 1 1 0 1>}.RGB)
   B=@n89
  )
  n91 = Subtract(
   A=Distance(A=ComponentMask(None=@n87) {g=true, r=true}, B=ComponentMask(None=@n90) {g=true, r=true})
   B=ScalarParameter() {defaultValue=84.8553, group=Boat Foam Mask, parameterName=SizeBoat_Offset}
  )
  n92 = ScalarParameter() {defaultValue=200, group=Boat Foam Mask, parameterName=SizeBoatBlend}
  NamedRerouteDeclaration(None=Saturate(None=Divide(A=@n91, B=@n92))) {name=Foam Masking, nodeColor=<rgba 0 1 0.1328 1>}
  NamedRerouteDeclaration(None=StaticSwitchParameter(True=@n85.G, False=@n86.FoamAlphaBaisB_Multiply) {group=Cascade, parameterName="OpenDPCurveControl(Cascade)"}) {name=DP2_FoamAlphaBias_Multiply, nodeColor=<rgba 0.8984 1 0 1>}
  NamedRerouteDeclaration(None=StaticSwitchParameter(True=@n85.B, False=@n86.MasterAlpha) {group=Cascade, parameterName="OpenDPCurveControl(Cascade)"}) {name=DP2_Master Alpha, nodeColor=<rgba 0.2188 0 1 1>}
  NamedRerouteDeclaration(None=StaticSwitchParameter(True=@n85.A, False=@n86.MasterAlphaContrast) {group=Cascade, parameterName="OpenDPCurveControl(Cascade)"}) {name=DP2_Master Alpha Contrast, nodeColor=<rgba 0.0312 1 0 1>}
  n93 = Constant()
  n94 = DynamicParameter() {
   parameterIndex=1
   defaultValue=<rgba 0 0 512 0>
   paramNames [Foam WP TextureCenterR, Foam WP TextureCenterG, Foam WP TextureScale, HSVOffset_Hue]
  }
  NamedRerouteDeclaration(None=StaticSwitchParameter(True=@n93, False=@n94.Foam WP TextureCenterR) {group=Cascade, parameterName="OpenDPCurveControl(Cascade)"}) {name=DP1_Foam WP TextureCenterR, nodeColor=<rgba 0.5937 1 0 1>}
  NamedRerouteDeclaration(None=StaticSwitchParameter(True=@n93, False=@n94.Foam WP TextureCenterG) {group=Cascade, parameterName="OpenDPCurveControl(Cascade)"}) {name=DP1_Foam WP TextureCenterG, nodeColor=<rgba 0.8516 0 1 1>}
  NamedRerouteDeclaration(None=StaticSwitchParameter(True=ScalarParameter() {defaultValue=512, parameterName=Tiling_Foam}, False=@n94.Foam WP TextureScale) {group=Cascade, parameterName="OpenDPCurveControl(Cascade)"}) {name=DP1_Foam WP TextureScale, nodeColor=<rgba 1 0 0.375 1>}
  NamedRerouteDeclaration(None=StaticSwitchParameter(True=@n56.R, False=@n57.Detail Time Control / Cascade - WaterBodyIndex) {group=Cascade, parameterName="OpenDPCurveControl(Cascade)"}) {name=Detail_Time_Control, nodeColor=<rgba 0 0.0156 1 1>}
 ]
}
