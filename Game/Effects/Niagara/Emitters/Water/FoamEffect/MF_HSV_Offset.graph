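# Material function, exposed to the library, that shifts an RGB color by an HSV offset, with optional tinting for darkening.
# It adds the offset to the color in HSV space and converts back through MF_FX_HSVtoRGB.
# Negative V also multiplies the result by a lerp from white toward a shifted reference color, weighted by absolute V.
# That reference shift runs the offset itself through RgbToHsv before use.
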MaterialFunction /Game/Effects/Niagara/Emitters/Water/FoamEffect/MF_HSV_Offset
graph {
 expressions {
  MaterialExpressionAbs_0 {class=MaterialExpressionAbs, inputs [None {from=MaterialExpressionComponentMask_45, output=""}], properties {}}
  MaterialExpressionAdd_2 {
   class=MaterialExpressionAdd
   inputs [
    A {from=MaterialExpressionRgbToHsv_2}
    B {from=MaterialExpressionFunctionInput_0}
   ]
  }
  MaterialExpressionAdd_3 {
   class=MaterialExpressionAdd
   inputs [
    A {from=MaterialExpressionRgbToHsv_3}
    B {from=MaterialExpressionConvert_0}
   ]
  }
  MaterialExpressionComponentMask_45 {class=MaterialExpressionComponentMask, inputs [None {from=MaterialExpressionReroute_1}], properties {b=true, desc=HSV Value}}
  MaterialExpressionConstant3Vector_0 {class=MaterialExpressionConstant3Vector}
  MaterialExpressionConstant3Vector_1 {class=MaterialExpressionConstant3Vector, properties {constant=<rgba 1 1 1 0>}}
  MaterialExpressionConstant3Vector_2 {class=MaterialExpressionConstant3Vector, properties {desc=Reference color to apply HSV offset for tinting, constant=<rgba 0.25 1 1 0>}}
  MaterialExpressionConstant_0 {class=MaterialExpressionConstant}
  MaterialExpressionConvert_0 {
   class=MaterialExpressionConvert
   inputs [
    None {from=MaterialExpressionMaterialFunctionCall_1, output=RGB}
    None {from=MaterialExpressionMaterialFunctionCall_1, output=RGB}
    None {from=""}
   ]
   properties {
    nodeName=Make Vector3
    convertInputs [
     {
      type=Scalar
      expressionInput {
       expression=&self:MaterialExpressionMaterialFunctionCall_1
       inputName=None
       mask=0
       maskA=0
       maskB=0
       maskG=0
       maskR=0
       outputIndex=0
      }
     }
     {
      type=Scalar
      expressionInput {
       expression=&self:MaterialExpressionMaterialFunctionCall_1
       inputName=None
       mask=0
       maskA=0
       maskB=0
       maskG=0
       maskR=0
       outputIndex=1
      }
     }
     {type=Scalar, expressionInput {expression=null, inputName=None, mask=0, maskA=0, maskB=0, maskG=0, maskR=0, outputIndex=2}}
    ]
    convertOutputs [{type=Vector3}]
   }
  }
  MaterialExpressionFunctionInput_0 {
   class=MaterialExpressionFunctionInput
   inputs [Preview {from=MaterialExpressionConstant3Vector_0, output=""}]
   properties {
    bUsePreviewValueAsDefault=true
    desc="Hue is -.5 to .5 range.  \r\n\r\nDropping V below 0 will tint (mult) the output color (ie \"negative V darkens\")."
    inputName=HSV Offset
   }
  }
  MaterialExpressionFunctionInput_1 {class=MaterialExpressionFunctionInput, inputs [Preview {from="", output=""}], properties {inputName=RGB}}
  MaterialExpressionFunctionInput_2 {
   class=MaterialExpressionFunctionInput
   inputs [Preview {from=MaterialExpressionConstant3Vector_2, output=""}]
   properties {
    bUsePreviewValueAsDefault=true
    description="Used if V is negative, will tint (mult) the output color"
    inputName=Optional Reference Color
    sortPriority=3
    previewValue=<0.25 1 1 1>
   }
  }
  MaterialExpressionFunctionOutput_0 {class=MaterialExpressionFunctionOutput, inputs [None {from=MaterialExpressionMultiply_0, output=""}]}
  MaterialExpressionHsvToRgb_2 {class=MaterialExpressionHsvToRgb, inputs [None {from=MaterialExpressionAdd_2, output=""}], properties {}}
  MaterialExpressionHsvToRgb_3 {class=MaterialExpressionHsvToRgb, inputs [None {from=MaterialExpressionAdd_3, output=""}], properties {}}
  MaterialExpressionIf_0 {
   class=MaterialExpressionIf
   inputs [
    A {from=MaterialExpressionComponentMask_45, output=""}
    B {from="", output=""}
    "A > B" {from=MaterialExpressionConstant_0, output=""}
    "A == B" {from=MaterialExpressionConstant_0, output=""}
    "A < B" {from=MaterialExpressionAbs_0, output=""}
   ]
   properties {}
  }
  MaterialExpressionLinearInterpolate_0 {
   class=MaterialExpressionLinearInterpolate
   inputs [
    A {from=MaterialExpressionConstant3Vector_1}
    B {from=MaterialExpressionHsvToRgb_3}
    Alpha {from=MaterialExpressionIf_0}
   ]
  }
  MaterialExpressionMaterialFunctionCall_1 {
   class=MaterialExpressionMaterialFunctionCall
   inputs [Color {from=MaterialExpressionRgbToHsv_4}]
   properties {materialFunction=&SplitComponents}
  }
  MaterialExpressionMaterialFunctionCall_3 {
   class=MaterialExpressionMaterialFunctionCall
   inputs [HSV {from=MaterialExpressionAdd_2}]
   properties {materialFunction=/Game/Effects/Niagara/MasterMaterials/Functions/MF_FX_HSVtoRGB}
  }
  MaterialExpressionMultiply_0 {
   class=MaterialExpressionMultiply
   inputs [
    A {from=MaterialExpressionReroute_0}
    B {from=MaterialExpressionReroute_3}
   ]
  }
  MaterialExpressionReroute_0 {class=MaterialExpressionReroute, inputs [None {from=MaterialExpressionMaterialFunctionCall_3, output=Result}]}
  MaterialExpressionReroute_1 {class=MaterialExpressionReroute, inputs [None {from=MaterialExpressionFunctionInput_0}]}
  MaterialExpressionReroute_3 {class=MaterialExpressionReroute, inputs [None {from=MaterialExpressionLinearInterpolate_0}]}
  MaterialExpressionRgbToHsv_2 {class=MaterialExpressionRgbToHsv, inputs [None {from=MaterialExpressionFunctionInput_1, output=""}], properties {}}
  MaterialExpressionRgbToHsv_3 {class=MaterialExpressionRgbToHsv, inputs [None {from=MaterialExpressionFunctionInput_2, output=""}], properties {}}
  MaterialExpressionRgbToHsv_4 {class=MaterialExpressionRgbToHsv, inputs [None {from=MaterialExpressionReroute_1, output=""}], properties {}}
 }
 properties {bExposeToLibrary=true, description=Shift HSV based on input values.}
}
